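# Physics asset for the Chip vehicle body skeletal mesh (LOD0, in an Old folder), giving bones simple collision bodies and joints.
# Each body is a single capsule, from the large main body to boost panels, spoiler, wheels and suspension parts; only the right rear brake pad has one.
# Constraints share uniform angular limits: panels and spoiler join body, front wheels chain through steering and differential to frame, and rear parts join frame directly.
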PhysicsAsset /VehicleCosmetics/Bodies/Chip/Meshes/Old/SK_Chip_LOD0_PhysicsAsset
bodies {
 body [{shapeType=Capsule, length=270.0567, radius=141.4514, center=<35.0711 0.0024 26.8046>, rotation=<rot 0 -89.9993 -84.0286>}]
 boost_panel_l [{shapeType=Capsule, length=55.1878, radius=11.5182, center=<-53.232 2.1724 2.7169>, rotation=<rot 0 -98.3695 -89.2295>}]
 boost_panel_r [{shapeType=Capsule, length=55.1879, radius=11.5182, center=<-53.232 -2.1724 2.7169>, rotation=<rot 0 -81.6305 -89.2295>}]
 boost_panel_up [{shapeType=Capsule, length=30.6983, radius=16.336, center=<-15.855 0.1527 -9.534>, rotation=<rot 0 179.4092 -89.9058>}]
 brakepad_bk_r [{shapeType=Capsule, length=17.7035, radius=13.3766, center=<-5.228 -1.4553 4.5054>, rotation=<rot 0 88.794 -138.3787>}]
 frame [{shapeType=Capsule, length=0.505, radius=0.505, rotation=<rot 0 -90 0>}]
 front_differential [{shapeType=Capsule, length=61.2796, radius=8.2225, center=<0.2931 -0.006 0.0599>, rotation=<rot 0 179.9994 -89.9993>}]
 shock_base_bk_l [{shapeType=Capsule, length=31.9748, radius=27.8492, center=<-0.6884 1.247 4.7596>, rotation=<rot 0 -98.1256 -83.1552>}]
 shock_base_bk_r [{shapeType=Capsule, length=31.9023, radius=22.8206, center=<0.2743 -5.2139 4.7311>, rotation=<rot 0 -83.7778 -83.1293>}]
 spoiler [{shapeType=Capsule, length=65.6521, radius=13.938, center=<-3.6085 0 -1.9047>, rotation=<rot 0 0 -90>}]
 wheel_disc_bk_l [{shapeType=Capsule, length=23.6882, radius=23.6858, center=<0.2932 -1.0964 0.0781>, rotation=<rot 0.0914 -90 17.0731>}]
 wheel_disc_bk_r [{shapeType=Capsule, length=23.6882, radius=23.6858, center=<0.2932 1.0964 0.0781>, rotation=<rot -0.0914 -90 17.0731>}]
 wheel_disc_fr_l [{shapeType=Capsule, length=23.6882, radius=23.6858, center=<0.2932 -2.3777 0.0781>, rotation=<rot 0.0914 -90 17.0733>}]
 wheel_disc_fr_r [{shapeType=Capsule, length=23.6882, radius=23.6858, center=<0.2932 2.3777 0.0781>, rotation=<rot -0.0914 -90 17.0734>}]
 wheel_steering_fr_l [{shapeType=Capsule, length=18.6884, radius=13.5814, center=<-5.1207 2.8034 6.1184>, rotation=<rot 0 -59.5748 -51.32>}]
 wheel_steering_fr_r [{shapeType=Capsule, length=18.6884, radius=13.5814, center=<-5.1207 -2.8034 6.1184>, rotation=<rot 0 -120.4252 -51.32>}]
}
constraints [
 {bone1Name=body, bone2Name=frame, swing1LimitDegrees=45, swing2LimitDegrees=45, twistLimitDegrees=45}
 {bone1Name=boost_panel_up, bone2Name=body, swing1LimitDegrees=45, swing2LimitDegrees=45, twistLimitDegrees=45}
 {bone1Name=boost_panel_l, bone2Name=body, swing1LimitDegrees=45, swing2LimitDegrees=45, twistLimitDegrees=45}
 {bone1Name=boost_panel_r, bone2Name=body, swing1LimitDegrees=45, swing2LimitDegrees=45, twistLimitDegrees=45}
 {bone1Name=spoiler, bone2Name=body, swing1LimitDegrees=45, swing2LimitDegrees=45, twistLimitDegrees=45}
 {bone1Name=front_differential, bone2Name=frame, swing1LimitDegrees=45, swing2LimitDegrees=45, twistLimitDegrees=45}
 {bone1Name=wheel_steering_fr_l, bone2Name=front_differential, swing1LimitDegrees=45, swing2LimitDegrees=45, twistLimitDegrees=45}
 {bone1Name=wheel_disc_fr_l, bone2Name=wheel_steering_fr_l, swing1LimitDegrees=45, swing2LimitDegrees=45, twistLimitDegrees=45}
 {bone1Name=wheel_steering_fr_r, bone2Name=front_differential, swing1LimitDegrees=45, swing2LimitDegrees=45, twistLimitDegrees=45}
 {bone1Name=wheel_disc_fr_r, bone2Name=wheel_steering_fr_r, swing1LimitDegrees=45, swing2LimitDegrees=45, twistLimitDegrees=45}
 {bone1Name=wheel_disc_bk_l, bone2Name=frame, swing1LimitDegrees=45, swing2LimitDegrees=45, twistLimitDegrees=45}
 {bone1Name=shock_base_bk_l, bone2Name=frame, swing1LimitDegrees=45, swing2LimitDegrees=45, twistLimitDegrees=45}
 {bone1Name=wheel_disc_bk_r, bone2Name=frame, swing1LimitDegrees=45, swing2LimitDegrees=45, twistLimitDegrees=45}
 {bone1Name=brakepad_bk_r, bone2Name=frame, swing1LimitDegrees=45, swing2LimitDegrees=45, twistLimitDegrees=45}
 {bone1Name=shock_base_bk_r, bone2Name=frame, swing1LimitDegrees=45, swing2LimitDegrees=45, twistLimitDegrees=45}
]
